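# Sound cue for firework explosions in a Season 21 week 1 quest.
# A Random node picks one of twelve equally weighted explosion waves, and a Modulator varies its pitch and volume on each play.
# Overridden attenuation turns off distance volume falloff but keeps listener focus and distance-based reverb send settings, despite referencing a Papaya stage fireworks attenuation asset.
# It routes to the non-critical in-game UI sound class.
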SoundCue /FNGameplayCues/Sounds/Quests/SFX/S21_Fireworks/Quest_S21_W1_Q1_Fireworks_Explode_Cue
root=Modulator_0
nodes [
 {id=Modulator_0, children [Random_1], properties {pitchMax=1.22, pitchMin=0.68, volumeMax=1.2, volumeMin=0.6}}
 {
  id=Random_1
  children [
   WavePlayer_4
   WavePlayer_5
   WavePlayer_6
   WavePlayer_7
   WavePlayer_8
   WavePlayer_9
   WavePlayer_10
   WavePlayer_11
   WavePlayer_12
   WavePlayer_13
   WavePlayer_14
   WavePlayer_15
  ]
  properties {weights [1,1,1,1,1,1,1,1,1,1,1,1]}
 }
 {id=WavePlayer_10, properties {soundWaveAssetPtr=&/Quest_S21_W1_Q1_Firework_Explode_06}}
 {id=WavePlayer_11, properties {soundWaveAssetPtr=&/Quest_S21_W1_Q1_Firework_Explode_07}}
 {id=WavePlayer_12, properties {soundWaveAssetPtr=&/Quest_S21_W1_Q1_Firework_Explode_09}}
 {id=WavePlayer_13, properties {soundWaveAssetPtr=&/Quest_S21_W1_Q1_Firework_Explode_10}}
 {id=WavePlayer_14, properties {soundWaveAssetPtr=&/Quest_S21_W1_Q1_Firework_Explode_11}}
 {id=WavePlayer_15, properties {soundWaveAssetPtr=&/Quest_S21_W1_Q1_Firework_Explode_12}}
 {id=WavePlayer_4, properties {soundWaveAssetPtr=&/Quest_S21_W1_Q1_Firework_Explode_08}}
 {id=WavePlayer_5, properties {soundWaveAssetPtr=&/Quest_S21_W1_Q1_Firework_Explode_01}}
 {id=WavePlayer_6, properties {soundWaveAssetPtr=&/Quest_S21_W1_Q1_Firework_Explode_02}}
 {id=WavePlayer_7, properties {soundWaveAssetPtr=&/Quest_S21_W1_Q1_Firework_Explode_03}}
 {id=WavePlayer_8, properties {soundWaveAssetPtr=&/Quest_S21_W1_Q1_Firework_Explode_04}}
 {id=WavePlayer_9, properties {soundWaveAssetPtr=&/Quest_S21_W1_Q1_Firework_Explode_05}}
]
properties {
 attenuationSettings=/Game/Athena/Sounds/Events/Papaya/SFX/Papaya_Stage_Fireworks_Att
 bOverrideAttenuation=true
 maxDistance=2097152
 soundClassObject=&Fort_UI_InGame_NonCritical
 volumeMultiplier=1.2
 attenuationOverrides {
  bAttenuate=false
  bEnableFocusInterpolation=true
  bEnableListenerFocus=true
  nonFocusVolumeAttenuation=0.5
  reverbDistanceMax=40000
  reverbDistanceMin=4000
  reverbWetLevelMax=0.4
  reverbWetLevelMin=0.1
 }
}
